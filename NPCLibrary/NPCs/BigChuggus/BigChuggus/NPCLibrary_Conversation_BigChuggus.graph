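# Conversation database for the Big Chuggus NPC, defining its greeting when a player starts talking to it.
# The entry point leads to one speech node with a default line and prioritized contextual lines for first meetings, returning players and two Slurp-themed outfits.
# Context requirements gate each variant, and curve rows set its priority.
# The speech node then links out to the shared NPC Library quest and basic services conversation entries.
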ConversationDatabase /NPCLibrary/NPCs/BigChuggus/BigChuggus/NPCLibrary_Conversation_BigChuggus
entry_points [{tag=Conversation.Entry.BigChuggus, destinations [#1]}]
nodes [
 ConversationEntryPointNode_7 {
  id=#1
  class=ConversationEntryPointNode
  outputs [#2]
  properties {entryTag {tagName=Conversation.Entry.BigChuggus}}
 }
 Speech_C_3 {
  id=#2
  class=Speech_C
  outputs [#3, #4]
  properties {
   nodeName=""
   generalConfig {
    defaultMessage=CHUGGUS?
    contextualMessages [
     {
      message="NEW FRIEND!"
      contextRequirements [{requirement {class=FortControllerRequirement_IsFirstEverConversationWithNPC}}]
      priority {curve {rowName=Tandem.Speech.Priority.FirstEverInteraction}}
     }
     {
      message="FRIEND BACK! FRIEND BACK!"
      contextRequirements [{requirement {class=FortControllerRequirement_CheckTimeSinceLastPlayedMatch}}]
      priority {curve {rowName=Tandem.Speech.Priority.WhileSincePlayed}}
     }
     {
      message="SLURP!!!!!"
      contextRequirements [{requirement {
         class=FortControllerRequirement_HasCID
         properties {allowedCIds [/BRCosmetics/Athena/Items/Cosmetics/Characters/CID_570_Athena_Commando_M_SlurpMonster]}
        }}]
      priority {curve {rowName=Tandem.Speech.Priority.OutfitReaction}}
     }
     {
      message="CHUGGUS MY THING! GET OWN THING! CHUGGUS MARKET SATURATED!"
      contextRequirements [{requirement {
         class=FortControllerRequirement_HasCID
         properties {allowedCIds [/BRCosmetics/Athena/Items/Cosmetics/Characters/CID_708_Athena_Commando_M_SoldierSlurp]}
        }}]
      priority {curve {rowName=Tandem.Speech.Priority.OutfitReaction}}
     }
    ]
   }
  }
 }
 ConversationLinkNode_0 {id=#4, class=ConversationLinkNode, properties {remoteEntryTag {tagName=Conversation.Entry.BasicServices}}}
 ConversationLinkNode_1 {id=#3, class=ConversationLinkNode, properties {remoteEntryTag {tagName=Conversation.Entry.NPCLibrary.Quests}}}
]
